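# Sound cue for the hit sound of the Vacuum Tube bow's chain lightning.
# A Mixer layers two branches, each a Modulator over an equally weighted Random node: one picks from six shared electric impact waves, the other from four bow-specific chain lightning waves.
# It uses the trap sound class and attenuation settings rather than weapon ones, and references a three-maximum, stop-oldest concurrency set.
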SoundCue /Game/Sounds/Fort_Weapons/Bows/VacuumTube/Weapon_Bow_VacuumTube_ChainLightning_Hit
root=Mixer_0
nodes [
 {id=Mixer_0, children [Modulator_9,Modulator_0], properties {inputVolume [1,1]}}
 {id=Modulator_0, children [Random_1]}
 {id=Modulator_9, children [Random_0]}
 {id=Random_0, children [WavePlayer_0,WavePlayer_1,WavePlayer_2,WavePlayer_3,WavePlayer_4,WavePlayer_5], properties {weights [1,1,1,1,1,1]}}
 {id=Random_1, children [WavePlayer_6,WavePlayer_7,WavePlayer_8,WavePlayer_9], properties {weights [1,1,1,1]}}
 {id=WavePlayer_0, properties {soundWaveAssetPtr=/Game/Sounds/Fort_GamePlay_Sounds/Elemental_Layers/enemy_electric_medium_impact_06}}
 {id=WavePlayer_1, properties {soundWaveAssetPtr=/Game/Sounds/Fort_GamePlay_Sounds/Elemental_Layers/enemy_electric_medium_impact_01}}
 {id=WavePlayer_2, properties {soundWaveAssetPtr=/Game/Sounds/Fort_GamePlay_Sounds/Elemental_Layers/enemy_electric_medium_impact_02}}
 {id=WavePlayer_3, properties {soundWaveAssetPtr=/Game/Sounds/Fort_GamePlay_Sounds/Elemental_Layers/enemy_electric_medium_impact_03}}
 {id=WavePlayer_4, properties {soundWaveAssetPtr=/Game/Sounds/Fort_GamePlay_Sounds/Elemental_Layers/enemy_electric_medium_impact_04}}
 {id=WavePlayer_5, properties {soundWaveAssetPtr=/Game/Sounds/Fort_GamePlay_Sounds/Elemental_Layers/enemy_electric_medium_impact_05}}
 {id=WavePlayer_6, properties {soundWaveAssetPtr=&/Weapon_VacuumTube_Bow_ChainLightning_04}}
 {id=WavePlayer_7, properties {soundWaveAssetPtr=&/Weapon_VacuumTube_Bow_ChainLightning_01}}
 {id=WavePlayer_8, properties {soundWaveAssetPtr=&/Weapon_VacuumTube_Bow_ChainLightning_02}}
 {id=WavePlayer_9, properties {soundWaveAssetPtr=&/Weapon_VacuumTube_Bow_ChainLightning_03}}
]
properties {
 attenuationSettings=&Fort_Traps_Att
 maxDistance=6000
 soundClassObject=&Fort_Traps
 volumeMultiplier=1.1
 attenuationOverrides {falloffDistance=3200, attenuationShapeExtents {x=800}}
 concurrencyOverrides {resolutionRule=PreventNew}
 concurrencySet [/Game/Sounds/Concurrency_Rules/ThreeMaximum_StopOldest_VolumeScale]
}
